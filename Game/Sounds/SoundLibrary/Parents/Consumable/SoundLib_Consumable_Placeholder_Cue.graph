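# Placeholder SoundCue in the sound library's consumable parent folder, presumably standing in for consumable item sounds.
# The cue is empty: it has no root node and no sound nodes, so it has nothing to play.
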SoundCue /Game/Sounds/SoundLibrary/Parents/Consumable/SoundLib_Consumable_Placeholder_Cue
root=null
nodes []
